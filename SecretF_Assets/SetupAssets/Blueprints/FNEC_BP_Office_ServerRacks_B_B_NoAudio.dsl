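; Blueprint for an office server rack prop, named as a variant without audio.
; Its construction script overrides material slot 1 on the static mesh component with a generic server screen material instance from the Helios office environment.
; The script holds no other logic.
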(fn ConstructionScript
  (Rendering|Material|SetMaterial (Variables|Debug|GetStaticMeshComponent) 1 "/Game/Environments/Helios/Props/Office/Office_ServerRacks_A/Materials/MI_Office_ServerScreenGeneric_A.MI_Office_ServerScreenGeneric_A"))
