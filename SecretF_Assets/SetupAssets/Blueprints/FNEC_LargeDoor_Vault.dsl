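; Large multi-panel vault door Blueprint that opens on activate and closes on reset, extending a parent door class.
; From either end state it plays steam and waits on a timer before TL_CloseDoor runs, moving bottom, middle and side panels over staggered percentage ranges.
; Closing starts bottom push targeting, and the authority smashes players and vehicles still overlapping once shut.
; Interrupted moves offset door audio by progress, and locking is intentionally inert.
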(event EventBeginPlay
  (|CallParentFunction)
  (Variables|/Script/DragonCartDoorsRuntime.DragonCartDoorPushComponent|SetPushPointSceneComp (Variables|Default|GetBottomDoorPushPoint) (Variables|Default|GetDoorPush_B))
  (Variables|/Script/DragonCartDoorsRuntime.DragonCartDoorPushComponent|SetPushPointSceneComp (Variables|Default|GetLeftDoorPushPoint) (Variables|Default|GetDoorPush_L))
  (Variables|/Script/DragonCartDoorsRuntime.DragonCartDoorPushComponent|SetPushPointSceneComp (Variables|Default|GetMiddleDoorPushPoint) (Variables|Default|GetDoorPush_M))
  (Variables|/Script/DragonCartDoorsRuntime.DragonCartDoorPushComponent|SetPushPointSceneComp (Variables|Default|GetRightDoorPushPoint) (Variables|Default|GetDoorPush_R)))

(event EventBP_OnActive
  (params (EnableAlternateActive :type Boolean) (EnableOpenCues :type Boolean))
  (|CallParentFunction EnableAlternateActive EnableOpenCues)
  (Utilities|Time|ClearandInvalidateTimerbyHandle (Variables|Default|GetCloseDoorDelayHandle)) ; This shouldn't matter, but better safe than sorry.
  (if (== (Variables|Default|GetDoorOperationPercentage) 0)
    (if (not (Utilities|Time|IsTimerActivebyHandle (Variables|Default|GetOpenDoorDelayHandle)))
      (Components|Activation|Activate (Variables|Default|GetOpeningSteamNS) false)
      (bind _returnvalue (Utilities|Time|SetTimerbyEvent (EventDispatchers|CreateEvent :!event Custom|OnActive_OpenDoor) (Variables|Default|GetActivateInitialDelay)))
      (Variables|Default|SetOpenDoorDelayHandle _returnvalue))
    (else
      (CallFunction|OnActiveOpenDoor))))

(event Default|EventBP_OnReset
  (params (EnableAlternateReset :type Boolean) (EnableOpenCues :type Boolean))
  (|CallParentFunction EnableAlternateReset EnableOpenCues)
  (Utilities|Time|ClearandInvalidateTimerbyHandle (Variables|Default|GetOpenDoorDelayHandle)) ; This shouldn't matter, but better safe than sorry.
  (if (>= (Variables|Default|GetDoorOperationPercentage) 1.0)
    (if (not (Utilities|Time|IsTimerActivebyHandle (Variables|Default|GetCloseDoorDelayHandle)))
      (Components|Activation|Activate (Variables|Default|GetOpeningSteamNS) false)
      (bind _returnvalue (Utilities|Time|SetTimerbyEvent (EventDispatchers|CreateEvent :!event Custom|OnReset_CloseDoor) (Variables|Default|GetActivateInitialDelay)))
      (Variables|Default|SetCloseDoorDelayHandle _returnvalue))
    (else
      (CallFunction|OnResetCloseDoor))))

(event Custom|OnActive_OpenDoor
  (bind (_direction _alpha) (Timeline|TL_CloseDoor Play)
    (:Update
      (Variables|Default|SetDoorOperationPercentage _alpha))
    (:Finished
      (CallFunction|OnDoorTimelineFinished _direction))))

(event Custom|OnReset_CloseDoor
  (Class|DragonCartDoorPushComponent|StartImpulseAsyncTargeting (Variables|Default|GetDoorPush_B))
  (Timeline|TL_CloseDoor Reverse))

(event OnActorTargeted(DoorPush_B)
  (params (TargetedActor :type Actor%20Object%20Reference))
  (CallFunction|HandleTarget TargetedActor))

(event OnActorTargeted(DoorPush_L)
  (params (TargetedActor :type Actor%20Object%20Reference))
  (CallFunction|HandleTarget TargetedActor))

(event OnActorTargeted(DoorPush_M)
  (params (TargetedActor :type Actor%20Object%20Reference))
  (CallFunction|HandleTarget TargetedActor))

(event OnActorTargeted(DoorPush_R)
  (params (TargetedActor :type Actor%20Object%20Reference))
  (CallFunction|HandleTarget TargetedActor))

(event MinigameEndedDelegate(FortMinigameLogic)
  (Utilities|Time|ClearandInvalidateTimerbyHandle (Variables|Default|GetCloseDoorDelayHandle))
  (Utilities|Time|ClearandInvalidateTimerbyHandle (Variables|Default|GetOpenDoorDelayHandle)))

(fn GetCollisionStaticMesh
  (return "0, 0, 0" (Variables|Default|GetCollisionMesh)))

(fn GetComponentToLock
  ;[1 1 1] This function just needs something returned in order for the Open and Close functions to work. The Lock functionality intentionally does nothing to this actor.
  (return (Utilities|Array|MakeArray (Variables|Default|GetMiddleDoorPanel))))

(fn GetMeshComponents
  (return 0))

(fn IgnoreWorldGridScaleOverride
  (return false))

(fn MoveDoorByPercentage
  (params (Percentage :type Float%20%28double-precision%29))
  (bind _returnvalue (Math|Float|NormalizetoRange Percentage 0.25 1.0))
  (bind _returnvalue_2 (Math|Float|Clamp(Float) _returnvalue))
  (|CallParentFunction Percentage)
  (Transformation|SetRelativeLocation (Variables|Default|GetBottomDoorPanel) (Math|Vector|MakeVector (.x (Variables|/Script/Engine.SceneComponent|GetRelativeLocation (Variables|Default|GetBottomDoorPanel))) (.y (Variables|/Script/Engine.SceneComponent|GetRelativeLocation (Variables|Default|GetBottomDoorPanel))) (Math|Float|Lerp (Variables|PanelPosition|GetBottomStartZ) (Variables|PanelPosition|GetBottomEndZ) (Math|Float|Clamp(Float) (Math|Float|NormalizetoRange Percentage 0.0 0.15)))) false false)
  (Transformation|SetRelativeLocation (Variables|Default|GetMiddleDoorPanel) (Math|Vector|MakeVector (.x (Variables|/Script/Engine.SceneComponent|GetRelativeLocation (Variables|Default|GetMiddleDoorPanel))) (.y (Variables|/Script/Engine.SceneComponent|GetRelativeLocation (Variables|Default|GetMiddleDoorPanel))) (Math|Float|Lerp (Variables|PanelPosition|GetMiddleStartZ) (Variables|PanelPosition|GetMiddleEndZ) (Math|Float|Clamp(Float) (Math|Float|NormalizetoRange Percentage 0.15 1.0)))) false false)
  (Transformation|SetRelativeLocation (Variables|Default|GetRightDoorPanel) (Math|Vector|MakeVector (Math|Float|Lerp (Variables|PanelPosition|GetRightStartX) (Variables|PanelPosition|GetRightEndX) _returnvalue_2) (.y (Variables|/Script/Engine.SceneComponent|GetRelativeLocation (Variables|Default|GetRightDoorPanel))) (.z (Variables|/Script/Engine.SceneComponent|GetRelativeLocation (Variables|Default|GetRightDoorPanel)))) false false)
  (Transformation|SetRelativeLocation (Variables|Default|GetLeftDoorPanel) (Math|Vector|MakeVector (Math|Float|Lerp (Variables|PanelPosition|GetLeftStartX) (Variables|PanelPosition|GetLeftEndX) _returnvalue_2) (.y (Variables|/Script/Engine.SceneComponent|GetRelativeLocation (Variables|Default|GetLeftDoorPanel))) (.z (Variables|/Script/Engine.SceneComponent|GetRelativeLocation (Variables|Default|GetLeftDoorPanel)))) false false))

(fn OnDoorTimelineFinished
  (params (Direction :type ETimelineDirection%20Enum :default "Forward"))
  (bind _overlapshape (Variables|Default|GetOverlapShape))
  (Class|DragonCartDoorPushComponent|StopImpulseAsyncTargeting (Variables|Default|GetDoorPush_B))
  (Networking|SwitchHasAuthority
    (:Authority
      (switch Utilities|FlowControl|Switch|SwitchonETimelineDirection (Variables|GetDirection)
        (:Forward
          _)
        (:Backward
          (CallFunction|SmashPlayers (Collision|GetOverlappingActors _overlapshape "/Script/FortniteGame.FortPlayerPawn" :!class /Script/Engine.PrimitiveComponent)) ; Smash players who are somehow still overlapping the closed door
          (CallFunction|SmashVehicles (Collision|GetOverlappingActors _overlapshape "/Script/FortniteGame.FortAthenaVehicle" :!class /Script/Engine.PrimitiveComponent)) ; Smash vehicles that are somehow still overlapping the closed door
        ))
      (|CallParentFunction (Variables|GetDirection)))))

(fn PlayDoorAudio
  (params (bOpen :type Boolean))
  (bind _dooraudio (Variables|Default|GetDoorAudio))
  (bind _dooraudio_2 (Variables|Default|GetDoorAudio))
  (if bOpen
    (if (== (Variables|Default|GetDoorOperationPercentage) 0)
      (|CallParentFunction true)
      (else
        (Audio|Components|Audio|SetSound _dooraudio_2 (Variables|Default|GetOpenSound))
        (Audio|Components|Audio|Play _dooraudio_2 (Math|Float|Max(Float) (* (Variables|Default|GetDoorOperationPercentage) (Variables|Default|GetTimelineLength)) 1.0))))
    (else
      (if (>= (Variables|Default|GetDoorOperationPercentage) 1.0)
        (|CallParentFunction false)
        (else
          (Audio|Components|Audio|SetSound _dooraudio (Variables|Default|GetCloseSound))
          (Audio|Components|Audio|Play _dooraudio (Math|Float|Max(Float) (* (- 1.0 (Variables|Default|GetDoorOperationPercentage)) (Variables|Default|GetTimelineLength)) 1.0)))))))

(fn SetTimelineLength
  (params (InTimeline :type Timeline%20Component%20Object%20Reference))
  (|CallParentFunction (Variables|FNEC_LargeDoor_Vault|GetTL_CloseDoor)))

(fn ConstructionScript
  (|CallParentFunction))
