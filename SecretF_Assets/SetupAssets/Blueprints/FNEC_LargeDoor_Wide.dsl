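; Child Blueprint for a wide three-panel door that lerps its left and right panels along X and its middle panel along Z.
; Activation plays a timeline driving the operation percentage, while reset reverses it and starts impulse targeting on the left push component.
; All three push components route targeted actors to HandleTarget.
; When the timeline finishes, left targeting stops, and after closing the server smashes overlapping players and vehicles.
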(event EventBeginPlay
  (|CallParentFunction)
  (Variables|/Script/DragonCartDoorsRuntime.DragonCartDoorPushComponent|SetPushPointSceneComp (Variables|Default|GetLeftDoorPushPoint) (Variables|Default|GetDoorPush_L))
  (Variables|/Script/DragonCartDoorsRuntime.DragonCartDoorPushComponent|SetPushPointSceneComp (Variables|Default|GetRightDoorPushPoint) (Variables|Default|GetDoorPush_R))
  (Variables|/Script/DragonCartDoorsRuntime.DragonCartDoorPushComponent|SetPushPointSceneComp (Variables|Default|GetMiddleDoorPushPoint) (Variables|Default|GetDoorPush_M)))

(event EventBP_OnActive
  (params (EnableAlternateActive :type Boolean) (EnableOpenCues :type Boolean))
  (|CallParentFunction EnableAlternateActive EnableOpenCues)
  (bind (_direction _alpha) (Timeline|TL_CloseDoor_Alternate Play)
    (:Update
      (Variables|Default|SetDoorOperationPercentage _alpha))
    (:Finished
      (CallFunction|OnDoorTimelineFinished _direction))))

(event Default|EventBP_OnReset
  (params (EnableAlternateReset :type Boolean) (EnableOpenCues :type Boolean))
  (|CallParentFunction EnableAlternateReset EnableOpenCues)
  (Class|DragonCartDoorPushComponent|StartImpulseAsyncTargeting (Variables|Default|GetDoorPush_L))
  (Timeline|TL_CloseDoor_Alternate Reverse))

(event OnActorTargeted(DoorPush_L)
  (params (TargetedActor :type Actor%20Object%20Reference))
  (CallFunction|HandleTarget TargetedActor))

(event OnActorTargeted(DoorPush_M)
  (params (TargetedActor :type Actor%20Object%20Reference))
  (CallFunction|HandleTarget TargetedActor))

(event OnActorTargeted(DoorPush_R)
  (params (TargetedActor :type Actor%20Object%20Reference))
  (CallFunction|HandleTarget TargetedActor))

(fn GetComponentToLock
  ;[1 1 1] This function just needs something returned in order for the Open and Close functions to work. The Lock functionality intentionally does nothing to this actor.
  (return (Utilities|Array|MakeArray (Variables|Default|GetMiddleDoorPanel))))

(fn MoveDoorByPercentage
  (params (Percentage :type Float%20%28double-precision%29))
  (|CallParentFunction Percentage)
  (Transformation|SetRelativeLocation (Variables|Default|GetRightDoorPanel) (Math|Vector|MakeVector (Math|Float|Lerp (Variables|PanelPosition|GetRightStartX) (Variables|PanelPosition|GetRightEndX) Percentage) (.y (Variables|/Script/Engine.SceneComponent|GetRelativeLocation (Variables|Default|GetRightDoorPanel))) (.z (Variables|/Script/Engine.SceneComponent|GetRelativeLocation (Variables|Default|GetRightDoorPanel)))) false false)
  (Transformation|SetRelativeLocation (Variables|Default|GetLeftDoorPanel) (Math|Vector|MakeVector (Math|Float|Lerp (Variables|PanelPosition|GetLeftStartX) (Variables|PanelPosition|GetLeftEndX) Percentage) (.y (Variables|/Script/Engine.SceneComponent|GetRelativeLocation (Variables|Default|GetLeftDoorPanel))) (.z (Variables|/Script/Engine.SceneComponent|GetRelativeLocation (Variables|Default|GetLeftDoorPanel)))) false false)
  (Transformation|SetRelativeLocation (Variables|Default|GetMiddleDoorPanel) (Math|Vector|MakeVector (.x (Variables|/Script/Engine.SceneComponent|GetRelativeLocation (Variables|Default|GetMiddleDoorPanel))) (.y (Variables|/Script/Engine.SceneComponent|GetRelativeLocation (Variables|Default|GetMiddleDoorPanel))) (Math|Float|Lerp (Variables|PanelPosition|GetMiddleStartZ) (Variables|PanelPosition|GetMiddleEndZ) Percentage)) false false))

(fn OnDoorTimelineFinished
  (params (Direction :type ETimelineDirection%20Enum :default "Forward"))
  (bind _overlapshape (Variables|Default|GetOverlapShape))
  (Class|DragonCartDoorPushComponent|StopImpulseAsyncTargeting (Variables|Default|GetDoorPush_L))
  (Networking|SwitchHasAuthority
    (:Authority
      (switch Utilities|FlowControl|Switch|SwitchonETimelineDirection (Variables|GetDirection)
        (:Forward
          _)
        (:Backward
          (CallFunction|SmashPlayers (Collision|GetOverlappingActors _overlapshape "/Script/FortniteGame.FortPlayerPawn" :!class /Script/Engine.PrimitiveComponent)) ; Smash players who are somehow still overlapping the closed door
          (CallFunction|SmashVehicles (Collision|GetOverlappingActors _overlapshape "/Script/FortniteGame.FortAthenaVehicle" :!class /Script/Engine.PrimitiveComponent)) ; Smash vehicles that are somehow still overlapping the closed door
        ))
      (|CallParentFunction (Variables|GetDirection)))))

(fn SetTimelineLength
  (params (InTimeline :type Timeline%20Component%20Object%20Reference))
  (|CallParentFunction (Variables|FNEC_LargeDoor_Wide|GetTL_CloseDoor_Alternate)))

(fn ConstructionScript
  (|CallParentFunction))
